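; Base pawn for VividRazor minion NPCs that handles guard voice audio and the elimination rez-out cue.
; On non-authority instances, BeginPlay flags heavy audio if the voice component's soundbank name ends in "Henchman_1", then binds vehicle enter/exit events.
; Those events swap between Convoy and standard soundbanks for Henchman_1 or Henchman_2.
; UpdateRezzoutGC sets a gameplay tag whose OnRep copies it into the elimination rez sequence cue.
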(event EventBeginPlay
  (|CallParentFunction)
  (Networking|SwitchHasAuthority
    (:Remote
      (CallFunction|GetAudioGuardType)
      (CallFunction|InitGuardAudio))))

(event Collision|EventActorBeginOverlap
  (params (OtherActor :type Actor%20Object%20Reference))
  (|CallParentFunction OtherActor) ; This node is disabled and will not be called.\nDrag off pins to build functionality.
)

(event EventTick
  (params (DeltaSeconds :type Float%20%28single-precision%29))
  (|CallParentFunction DeltaSeconds) ; This node is disabled and will not be called.\nDrag off pins to build functionality.
)

(event Custom|UpdateRezzoutGC
  (params (RezzOutGC :type Gameplay%20Tag%20Structure))
  (Variables|Default|SetRezz%20Out%20GC RezzOutGC))

(fn GetAudioGuardType
  (bind _self self)
  (bind _returnvalue (Actor|GetComponentbyClass _self "/Game/Sounds/Fort_Audio_Blueprints/VoiceComponent/NPC_VoiceComponent.NPC_VoiceComponent_C"))
  (Utilities|IsValid _returnvalue
    (:Is%20Valid
      (if (Utilities|String|EndsWith (Utilities|GetObjectName (Variables|/Script/FortniteAI.FortAIComponent_Voice|GetVoiceSoundBank _returnvalue)) "Henchman_1")
        (Variables|Default|SetbIsHeavyAudio true)
        (else
          (Variables|Default|SetbIsHeavyAudio false))))))

(fn InitGuardAudio
  (Default|BindEventtoOnEnteredVehicleDispatcher (EventDispatchers|CreateEvent :!event On%20Guard%20Enter%20Vehicle%20Audio))
  (Default|BindEventtoOnExitedVehicleDispatcher (EventDispatchers|CreateEvent :!event On%20Guard%20Exit%20Vehicle%20Audio)))

(fn OnGuardEnterVehicleAudio
  (bind _self self)
  (bind _returnvalue (Actor|GetComponentbyClass _self "/Game/Sounds/Fort_Audio_Blueprints/VoiceComponent/NPC_VoiceComponent.NPC_VoiceComponent_C"))
  (Utilities|IsValid _returnvalue
    (:Is%20Valid
      (Variables|/Script/FortniteAI.FortAIComponent_Voice|SetVoiceSoundBank (select (Variables|Default|GetbIsHeavyAudio) "/VividRazor_Minions/Audio/VO/Henchman_1/Soundbank_S34_Henchman_1_Convoy.Soundbank_S34_Henchman_1_Convoy" "/VividRazor_Minions/Audio/VO/Henchman_2/Soundbank_S34_Henchman_2_Convoy.Soundbank_S34_Henchman_2_Convoy") _returnvalue))))

(fn OnGuardExitVehicleAudio
  (bind _self self)
  (bind _returnvalue (Actor|GetComponentbyClass _self "/Game/Sounds/Fort_Audio_Blueprints/VoiceComponent/NPC_VoiceComponent.NPC_VoiceComponent_C"))
  (Utilities|IsValid _returnvalue
    (:Is%20Valid
      (Variables|/Script/FortniteAI.FortAIComponent_Voice|SetVoiceSoundBank (select (Variables|Default|GetbIsHeavyAudio) "/VividRazor_Minions/Audio/VO/Henchman_1/Soundbank_S34_Henchman_1.Soundbank_S34_Henchman_1" "/VividRazor_Minions/Audio/VO/Henchman_2/Soundbank_S34_Henchman_2.Soundbank_S34_Henchman_2") _returnvalue))))

(fn OnRep_RezzOutGC
  (Variables|Elimination|SetElimination%20Rez%20Sequence%20GC (Variables|Default|GetRezz%20Out%20GC)))

(fn ConstructionScript
  (|CallParentFunction))
